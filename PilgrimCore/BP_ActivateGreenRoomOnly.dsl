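; Blueprint actor that keeps a configured list of actors active only during the Quickplay Pregame state, which its name identifies as the green room.
; On BeginPlay it waits for the Pilgrim event router, then listens for Quickplay state Began and Ended events.
; Pregame beginning or ending calls Toggle, which activates (with reset) or deactivates the root component of each listed actor.
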(event EventBeginPlay
  (bind _eventrouter (Pilgrim|PilgrimWaitForEventRouter 0)
    (:OnEventRouterAvailable
      (bind (_asynctaskproxy _payload) (GameplayEventRouter|StartListeningToEvent _eventrouter "/Script/PilgrimQPSharedRuntime.PilgrimEvent_QuickplayState_Began")
        (:then
          (bind (_asynctaskproxy_2 _payload_2) (GameplayEventRouter|StartListeningToEvent _eventrouter "/Script/PilgrimQPSharedRuntime.PilgrimEvent_QuickplayState_Ended")
            (:OnEventReceived
              (switch Utilities|FlowControl|Switch|SwitchonEPilgrimQuickplayState (Utilities|Struct|BreakPilgrimEventQuickplayStateBase _payload_2)
                (:Pregame
                  (CallFunction|Toggle false))))))
        (:OnEventReceived
          (switch Utilities|FlowControl|Switch|SwitchonEPilgrimQuickplayState (Utilities|Struct|BreakPilgrimEventQuickplayStateBase _payload)
            (:Pregame
              (CallFunction|Toggle true))))))))

(event Collision|EventActorBeginOverlap
  (params (OtherActor :type Actor%20Object%20Reference)))

(event EventTick
  (params (DeltaSeconds :type Float%20%28single-precision%29)))

(fn Toggle
  (params (Activate :type Boolean))
  (for _array_element (Variables|Default|GetActors)
    (if Activate
      (Components|Activation|Activate (Variables|/Script/Engine.Actor|GetRootComponent _array_element) true)
      (else
        (Components|Activation|Deactivate (Variables|/Script/Engine.Actor|GetRootComponent _array_element))))))

(fn ConstructionScript)
