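# Translucent unlit spline-mesh material for sustain note trails on the Pilgrim music track.
# Custom primitive data supplies lane, overdrive, pro colors, type and modulation, alongside a color curve atlas and shared track-phase opacity from a parameter collection.
# Procedural UV masks build sustain body, lift trail, drum overdrive deploy, rolls-and-swells and end-cap looks, with chained saturated lerps selecting between four variants.
# A time-driven sine provides wobble offset.
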Material /PilgrimCore/Effects/M_Pilgrim_Sustain
graph {
 properties {
  bCastRayTracedShadows=false
  bUsedWithSplineMeshes=true
  blendMode=BLEND_Translucent
  pixelDepthOffsetMode=PDOM_Legacy
  shadingModel=MSM_Unlit
  usedShadingModels=Unlit
 }
 nodes [
  Constant() {r=3}
  n1 = Multiply(A=NamedRerouteUsage(), B=NamedRerouteUsage())
  n2 = Subtract(A=@n1) {constB=-0.5}
  LinearInterpolate(A=@n1, B=@n2, Alpha=NamedRerouteUsage())
  LinearInterpolate(
   A=Constant3Vector() {constant=<rgba 1.25 0.3765 0.2018 1>}
   B=Constant3Vector() {constant=<rgba 1 1 1 1>}
   Alpha=Multiply(A=NamedRerouteUsage(), B=NamedRerouteUsage()) {desc="IsOverdrive AND UseProColors\r\n(formerly IsPlasticOverdrive)"}
  )
  n3 = NamedRerouteUsage()
  n4 = Saturate(None=@n3)
  n5 = LinearInterpolate(A=Multiply(A=NamedRerouteUsage(), B=NamedRerouteUsage()), B=NamedRerouteUsage(), Alpha=@n4)
  n6 = NamedRerouteUsage()
  n7 = Saturate(None=Subtract(A=@n3))
  n8 = Saturate(None=Subtract(A=@n3) {constB=2})
  MaterialFunctionCall(Emissive=LinearInterpolate(A=LinearInterpolate(A=@n5, B=@n6, Alpha=@n7), B=NamedRerouteUsage(), Alpha=@n8)) {materialFunction=&MF_FX_EmissiveAdjustments}
  n9 = NamedRerouteUsage()
  n10 = Constant() {r=0.99}
  n11 = Multiply(A=ComponentMask(None=TextureCoordinate()) {r=true}) {constB=0.5}
  n12 = Add(A=SmoothStep(Max=@n10, Value=Sine(None=@n11)) {constMax=0.99, constMin=1}) {constB=0.1}
  n13 = Subtract(A=@n10) {constB=0.025}
  n14 = TextureCoordinate()
  n15 = Sine(None=Multiply(A=ComponentMask(None=@n14) {r=true}) {constB=0.5})
  n16 = MaterialFunctionCall(UV Channel=MaterialFunctionCall!(Repeat Count=Constant2Vector() {g=1, r=2}, texcoord=TextureCoordinate()) {materialFunction=/Game/Environments/Papaya/Materials/MaterialFunctions/MF_UVMirror}.Result) {materialFunction=&LinearGradient}
  n17 = Constant() {r=2.5}
  n18 = Add(
   A=Multiply(A=Saturate(None=@n12), B=SmoothStep(Min=@n13, Value=@n15) {constMin=0.975})
   B=Multiply(A=Power(Base=@n16.UGradient, Exp=@n17), B=NamedRerouteUsage())
  )
  n19 = LinearInterpolate(
   A=LinearInterpolate(A=NamedRerouteDeclaration(None=@n18) {name=Sustain Body, nodeColor=<rgba 0.2656 1 0 1>}, B=NamedRerouteUsage(), Alpha=NamedRerouteUsage())
   B=NamedRerouteUsage()
   Alpha=@n4
  )
  Multiply(A=NamedRerouteUsage(), B=LinearInterpolate(A=LinearInterpolate(A=@n19, B=@n9, Alpha=@n7), B=NamedRerouteUsage(), Alpha=@n8))
  n20 = MaterialFunctionCall(Number=NamedRerouteUsage(), MaximumNumberOfDigits=Constant() {r=4}, UVs=Add(A=TextureCoordinate(), B=Constant2Vector() {g=-0.25}) {constB=0}) {materialFunction=&DebugScalarValues}
  n21 = MaterialFunctionCall(Number=NamedRerouteUsage(), MaximumNumberOfDigits=Constant() {r=4}, UVs=Add(A=TextureCoordinate(), B=Constant2Vector() {g=0.25}) {constB=0}) {materialFunction=&DebugScalarValues}
  n22 = Step(Y=MaterialFunctionCall() {materialFunction=&LinearGradient}.VGradient) {constX=0.5}
  n23 = Multiply(A=NamedRerouteUsage(), B=Constant2Vector() {g=1, r=2})
  n24 = MaterialFunctionCall() {materialFunction=&LinearGradient}
  NamedRerouteDeclaration(None=LinearInterpolate(
    A=LinearInterpolate(A=@n20.Result, B=@n21.Result, Alpha=@n22)
    B=ComponentMask(None=@n23) {g=true}
    Alpha=Step(Y=@n24.UGradient) {constX=0.5}
   )) {name=UVTest, nodeColor=<rgba 0 0 0 1>}
  n25 = TextureCoordinate()
  n26 = OneMinus(None=@n25)
  n27 = Add(A=Multiply(A=ComponentMask(None=@n26) {g=true}, B=NamedRerouteUsage()), B=NamedRerouteUsage())
  n28 = NamedRerouteUsage()
  NamedRerouteDeclaration(None=AppendVector(A=ComponentMask(None=@n25) {r=true}, B=Multiply(A=@n27, B=@n28))) {name=SustainUV, nodeColor=<rgba 0 0 0 1>}
  n29 = MaterialFunctionCall!(In=VectorParameter() {bUseCustomPrimitiveData=true, parameterName=PilgrimTrackItemStartIndex}.RGBA) {materialFunction=/PilgrimCore/GemAndTrack/Track/Materials/MF_GetPilgrimTrackItemParams}
  NamedRerouteDeclaration(None=@n29.LaneIndex) {name=LaneIndex, nodeColor=<rgba 0 1 0.0291 1>}
  NamedRerouteDeclaration(None=@n29.IsOverdrive) {name=IsOverdrive, nodeColor=<rgba 0 1 0.0291 1>}
  NamedRerouteDeclaration(None=@n29.UseProColors) {name=UseProColors, nodeColor=<rgba 0 1 0.0291 1>}
  n30 = NamedRerouteUsage()
  NamedRerouteDeclaration(None=LinearInterpolate(B=Add(A=@n30), Alpha=NamedRerouteUsage()) {desc="Non-pro colors use index 0\r\nPro colors use lane index shifted up by 1, to indices 1-5"}) {name=LaneColorIndex, nodeColor=<rgba 0 1 0.0291 1>}
  NamedRerouteDeclaration(None=ScalarParameter() {bUseCustomPrimitiveData=true, defaultValue=0.5, parameterName=UVStart, primitiveDataIndex=4}) {name=UVStart, nodeColor=<rgba 0 1 0.0291 1>}
  NamedRerouteDeclaration(None=ScalarParameter() {bUseCustomPrimitiveData=true, parameterName=UVLength, primitiveDataIndex=5}) {name=UVLength, nodeColor=<rgba 0 1 0.0291 1>}
  NamedRerouteDeclaration(None=ScalarParameter() {bUseCustomPrimitiveData=true, parameterName=TrackLength, primitiveDataIndex=6}) {name=TrackLength, nodeColor=<rgba 0 1 0.0291 1>}
  NamedRerouteDeclaration(None=ScalarParameter() {bUseCustomPrimitiveData=true, parameterName=Modulation, primitiveDataIndex=8}) {name=Modulation, nodeColor=<rgba 0 1 0.0291 1>}
  NamedRerouteDeclaration(None=ScalarParameter() {bUseCustomPrimitiveData=true, defaultValue=1, parameterName=Intensity, primitiveDataIndex=9}) {name=Intensity, nodeColor=<rgba 0 1 0.0291 1>}
  n31 = ComponentMask(None=MaterialFunctionCall() {materialFunction=&LocalPosition}."Local Position (Excluding Offsets)") {g=true}
  n32 = Constant() {r=0.005}
  n33 = Time()
  n34 = Constant() {r=-1.15}
  n35 = Multiply(
   A=Add(A=Multiply(A=@n31, B=@n32) {constB=0.005}, B=Multiply(A=@n33, B=@n34))
   B=LinearInterpolate(A=Constant() {r=3}, B=Constant() {r=-3}, Alpha=VertexColor().R)
  ) {constB=8}
  n36 = LinearInterpolate(A=Constant3Vector() {constant=<rgba 2 0 0 1>}, Alpha=NamedRerouteUsage()) {constB=0}
  n37 = NamedRerouteUsage()
  NamedRerouteDeclaration(None=Multiply(A=Sine(None=@n35), B=Multiply(A=@n36, B=@n37))) {name=WobbleWPO, nodeColor=<rgba 0 0 0 1>}
  NamedRerouteDeclaration(None=ScalarParameter() {bUseCustomPrimitiveData=true, parameterName=Glow, primitiveDataIndex=10}) {name=Glow, nodeColor=<rgba 0 1 0.0291 1>}
  NamedRerouteDeclaration(None=@n29.TrackIndex) {name=TrackIndex, nodeColor=<rgba 0 1 0.0291 1>}
  NamedRerouteDeclaration(None=MaterialFunctionCall!(
    Values=CollectionParameter() {collection=/SparksEffects/Effects/MPC_Sparks, parameterName=TrackSecondaryElementsPhase}
    Index=NamedRerouteUsage()
   ) {desc=Shared opacity from track, materialFunction=/PilgrimCore/GemAndTrack/Gems/Materials/MF_SelectComponentByIndex}.Value) {name=Alpha, nodeColor=<rgba 0 0 0 1>}
  n38 = MaterialFunctionCall(Float2=TextureCoordinate()) {materialFunction=&BreakOutFloat2Components}
  n39 = MaterialFunctionCall(Float2=TextureCoordinate()) {materialFunction=&BreakOutFloat2Components}
  n40 = LinearInterpolate(A=OneMinus(None=@n39.R), B=@n39.R, Alpha=Step(Y=@n39.R) {constX=0.5})
  n41 = Multiply(A=@n40) {constB=0.2}
  n42 = Sine(None=Multiply(A=Add(A=@n38.G, B=@n41)) {constB=3})
  n43 = Constant() {r=-1}
  n44 = Constant() {r=1}
  NamedRerouteDeclaration(None=Step(Y=MaterialFunctionCall!(Input=@n42, Input Low=@n43, Input High=@n44) {materialFunction=/Engine/Functions/Engine_MaterialFunctions03/Math/RemapValueRangeNormalized}.Result) {constX=0.5}) {name=LiftNoteTrailArrows, nodeColor=<rgba 0.5336 0 1 1>}
  n45 = NamedRerouteUsage()
  n46 = Constant() {r=1}
  n47 = Step(Y=Add(A=NamedRerouteUsage(), B=Divide(A=@n46) {constB=10}) {constB=0.03}) {constX=0.5}
  NamedRerouteDeclaration(None=Subtract(A=Step(Y=@n45) {constX=0.5}, B=@n47)) {name=LiftNoteTrailBorders, nodeColor=<rgba 0.5336 0 1 1>}
  n48 = Add(A=NamedRerouteUsage(), B=NamedRerouteUsage())
  NamedRerouteDeclaration(None=LinearInterpolate(A=Saturate(None=@n48), B=NamedRerouteUsage(), Alpha=NamedRerouteUsage())) {name=LiftNoteTrailOpacity, nodeColor=<rgba 0.5336 0 1 1>}
  n49 = NamedRerouteUsage()
  n50 = NamedRerouteUsage()
  n51 = NamedRerouteUsage()
  n52 = NamedRerouteDeclaration(None=Constant3Vector() {constant=<rgba 0.6724 0.6602 0.75 1>}) {name=TrailEndInnerColor, nodeColor=<rgba 1 0 0 1>}
  n53 = NamedRerouteUsage()
  n54 = Add(A=Multiply(A=NamedRerouteUsage(), B=@n53), B=@n53)
  n55 = Add(A=Add(A=Multiply(A=@n49, B=@n50), B=Multiply(A=@n51, B=@n52)), B=Multiply(A=Multiply(A=@n54) {constB=0.4}, B=NamedRerouteUsage()))
  n56 = Multiply(A=Multiply(A=@n55, B=NamedRerouteUsage())) {constB=5}
  NamedRerouteDeclaration(None=LinearInterpolate(A=Add(A=@n55, B=@n56), B=NamedRerouteUsage(), Alpha=NamedRerouteUsage())) {name=LiftNoteTrailColor, nodeColor=<rgba 0.5336 0 1 1>}
  NamedRerouteDeclaration(None=Multiply(A=@n47) {constB=0.5}) {name=LiftNoteTrailBody, nodeColor=<rgba 0.5336 0 1 1>}
  NamedRerouteDeclaration(None=ScalarParameter() {defaultValue=0.1, group=LiftNoteTrail, parameterName=LiftNoteTrailAngle, sliderMax=1}) {name=LiftNoteTrailBoderAngle, nodeColor=<rgba 0.5336 0 1 1>}
  NamedRerouteDeclaration(None=ScalarParameter() {bUseCustomPrimitiveData=true, parameterName=SustainType, primitiveDataIndex=7}) {name=SustainType, nodeColor=<rgba 0 1 0.0291 1>}
  n57 = NamedRerouteUsage()
  n58 = CurveAtlasRowParameter(CurveTime=Divide(A=Add(A=@n57) {constB=0.5}) {constB=6}) {
   atlas=/PilgrimCore/GemAndTrack/Colors/Sparks_Track_LUT
   curve=/PilgrimCore/GemAndTrack/Colors/Sparks_Track_Sustains
   defaultValue=2
   parameterName=SustainColor
  }
  NamedRerouteDeclaration(None=LinearInterpolate(A=@n58, B=Constant3Vector() {constant=<rgba 0.6724 0.6602 0.75 1>}, Alpha=NamedRerouteUsage())) {name=Color, nodeColor=<rgba 0 0 0 1>}
  n59 = Add(A=OneMinus(None=@n40)) {constB=0.5}
  n60 = Multiply(A=@n39.G, B=NamedRerouteUsage()) {constB=0.2}
  n61 = NamedRerouteUsage()
  NamedRerouteDeclaration(None=Add(A=Subtract(A=@n59, B=@n60), B=Add(B=@n61) {constA=-1}) {constB=-0.8}) {name=LiftNoteTrailBodyGradient, nodeColor=<rgba 0.5336 0 1 1>}
  n62 = NamedRerouteUsage()
  n63 = Step(Y=NamedRerouteUsage()) {constX=0.48}
  NamedRerouteDeclaration(None=Saturate(None=Subtract(A=@n62, B=@n63))) {name=LiftNoteTrailEdges, nodeColor=<rgba 0.5336 0 1 1>}
  n64 = Saturate(None=Power(Base=NamedRerouteUsage()) {constExponent=3.5})
  n65 = MaterialFunctionCall(Float2=TextureCoordinate()) {materialFunction=&BreakOutFloat2Components}
  n66 = Saturate(None=Power(Base=Add(A=@n65.G) {constB=0.5}) {constExponent=1.1})
  NamedRerouteDeclaration(None=LinearInterpolate(A=LinearInterpolate(A=@n52, B=@n64, Alpha=@n66), B=@n52, Alpha=NamedRerouteUsage())) {name=LiftNoteTrailBaseColor, nodeColor=<rgba 0.5336 0 1 1>}
  n67 = Add(A=@n39.G) {constB=-0.3}
  n68 = LinearInterpolate(A=OneMinus(None=@n67), B=@n67, Alpha=Step(Y=@n67) {constX=0.5})
  NamedRerouteDeclaration(None=Saturate(None=SmoothStep(Value=@n68) {constMax=0.65, constMin=0.35})) {name=LiftNoteTrailShine, nodeColor=<rgba 0.5336 0 1 1>}
  n69 = TextureCoordinate()
  n70 = TextureCoordinate()
  n71 = ComponentMask(None=OneMinus(None=@n70)) {g=true}
  n72 = Multiply(A=ComponentMask(None=TextureCoordinate()) {r=true}) {constB=0.5}
  n73 = Saturate(None=Power(Base=Sine(None=@n72), Exp=Constant() {r=2}) {constExponent=5})
  n74 = Multiply(
   A=TextureSample(UVs=Panner(Coordinate=@n69) {speedY=0.25}) {texture=/PilgrimCore/GemAndTrack/Track/Textures/T_Sparks_Track_PeakState}.G
   B=Multiply(A=Power(Base=@n71) {constExponent=2.5}, B=@n73)
  )
  n75 = Panner(Coordinate=TextureCoordinate() {vTiling=0.2}) {speedY=0.5}
  n76 = Power(Base=@n71) {constExponent=1}
  n77 = Multiply(
   A=Multiply(A=TextureSample(UVs=@n75) {texture=/Game/Effects/Fort_Effects/Emotes/Candor/T_Noise_Blur_Texture}.G) {constB=5}
   B=Multiply(A=Multiply(A=@n76, B=@n73)) {constB=0.5}
  )
  n78 = ComponentMask(None=@n70) {g=true}
  n79 = Power(Base=@n78) {constExponent=20}
  NamedRerouteDeclaration(None=LinearInterpolate(A=Add(A=@n74, B=@n77), B=Multiply(A=@n73, B=@n79), Alpha=NamedRerouteUsage())) {name=DrumDeployODOpacity, nodeColor=<rgba 0.5029 0 1 1>}
  n80 = NamedRerouteUsage()
  n81 = Add(A=@n80) {constB=0.25}
  n82 = TextureCoordinate()
  n83 = Power(Base=ComponentMask(None=OneMinus(None=@n82)) {g=true}, Exp=Constant() {r=10}) {constExponent=1}
  n84 = Constant() {r=5}
  n85 = Multiply(A=@n81, B=Add(A=Multiply(A=@n83, B=@n84)) {constB=0})
  n86 = Multiply(A=Power(Base=@n78) {constExponent=5}) {constB=5}
  n87 = Multiply(A=@n80, B=@n81)
  NamedRerouteDeclaration(None=LinearInterpolate(A=Add(A=@n80, B=@n85), B=Add(A=@n86, B=@n87), Alpha=NamedRerouteUsage())) {name=DrumDeployODColor, nodeColor=<rgba 0.5 0 1 1>}
  n88 = TextureCoordinate()
  NamedRerouteDeclaration(None=ComponentMask(None=@n88) {r=true}) {name=U, nodeColor=<rgba 0.2422 1 0 1>}
  n89 = ComponentMask(None=@n88) {g=true}
  NamedRerouteDeclaration(None=LinearInterpolate(A=OneMinus(None=@n89), B=@n89, Alpha=ScalarParameter() {bUseCustomPrimitiveData=true, parameterName=Flip End Caps, primitiveDataIndex=13})) {name=V, nodeColor=<rgba 1 0 0.5625 1>}
  NamedRerouteDeclaration(None=ScalarParameter() {defaultValue=1, group=End Cap, parameterName=Is Endcap, primitiveDataIndex=13, sortPriority=0}) {name=Is EndCap, nodeColor=<rgba 0.1016 1 0 1>}
  n90 = Add(A=OneMinus(None=NamedRerouteUsage()), B=Divide(A=ScalarParameter() {defaultValue=0.5, parameterName=Y Position}, B=NamedRerouteUsage())) {constB=0.08}
  n91 = Saturate(None=Multiply(A=Subtract(A=@n90)) {constB=1000})
  n92 = Divide(A=NamedRerouteUsage()) {constB=0.0855}
  n93 = Multiply(A=@n91, B=SmoothStep(Value=@n92) {constMax=0.4, constMin=1})
  n94 = SmoothStep(Value=Divide(A=NamedRerouteUsage()) {constB=0.06}) {constMax=0.16, constMin=1}
  n95 = Saturate(None=Multiply(A=NamedRerouteUsage(), B=Subtract(A=@n94) {constB=0.105}))
  n96 = SmoothStep(Value=NamedRerouteUsage()) {constMax=-0.45, constMin=0.5}
  n97 = Multiply(A=Multiply(A=NamedRerouteUsage(), B=Multiply(A=@n96)), B=NamedRerouteUsage())
  NamedRerouteDeclaration(None=Add(A=Subtract(A=@n93, B=@n95), B=Saturate(None=@n97))) {name=SustainEndCap, nodeColor=<rgba 1 0 0.4219 1>}
  n98 = MaterialFunctionCall() {materialFunction=&ObjectScale}
  NamedRerouteDeclaration(None=Divide(
    A=ChannelMaskParameter(None=@n98.Scale XYZ) {group=End Cap, maskChannel=Green, parameterName=Object Bounds X, defaultValue=<rgba 0 1 0 0>}.R
    B=ChannelMaskParameter(None=@n98.Scale XYZ) {group=End Cap, maskChannel=Blue, parameterName=Object Bounts Y, defaultValue=<rgba 0 0 1 0>}.R
   )) {name=Aspect Ratio, nodeColor=<rgba 0 1 0.7188 1>}
  n99 = Step(X=ComponentMask(None=NamedRerouteUsage()) {g=true}) {constX=0.8, constY=0.8}
  n100 = TextureCoordinate()
  n101 = LinearInterpolate(A=@n100, B=OneMinus(None=@n100), Alpha=ScalarParameter() {bUseCustomPrimitiveData=true, parameterName=Flip End Caps, primitiveDataIndex=13})
  n102 = SphereMask(
   A=MaterialFunctionCall(UVs=NamedRerouteDeclaration(None=@n101) {name=UVs, nodeColor=<rgba 0.9219 1 0 1>}, Texture Scale=Constant2Vector() {g=0.5, r=1.01}) {materialFunction=&ScaleUVsByCenter}.UVs
   B=Constant2Vector() {g=1.1, r=0.5}
  ) {attenuationRadius=0.5, hardnessPercent=0}
  n103 = Step(X=@n102) {constX=0.1, constY=0.01}
  n104 = Saturate(None=Add(A=@n99, B=@n103))
  NamedRerouteDeclaration(None=LinearInterpolate(A=Constant() {r=1}, B=@n104, Alpha=NamedRerouteUsage())) {name=RollsAndSwellsOpacity, nodeColor=<rgba 1 0.0938 0 1>}
  NamedRerouteDeclaration(None=ScalarParameter() {bUseCustomPrimitiveData=true, group=End Cap, parameterName=Use Rounded Ends, primitiveDataIndex=12, sortPriority=0}) {name=Use Rounded Ends, nodeColor=<rgba 1 0.9609 0 1>}
  n105 = NamedRerouteUsage()
  n106 = Add(A=Step(Y=NamedRerouteUsage()) {constX=0.15}, B=OneMinus(None=Step(Y=@n105) {constX=0.85}))
  NamedRerouteDeclaration(None=OneMinus(None=@n106)) {name=TrailEndInnerMask, nodeColor=<rgba 1 0.7031 0 1>}
  n107 = NamedRerouteUsage()
  n108 = Subtract(A=Multiply(A=NamedRerouteUsage(), B=NamedRerouteUsage())) {constB=-0.5}
  n109 = Distance(A=AppendVector(A=Subtract(A=@n107) {constB=0.31}, B=@n108), B=Constant2Vector() {g=0.5, r=0.5})
  n110 = SmoothStep(Value=Divide(A=NamedRerouteDeclaration(None=@n109) {name=Radial UV_TrailInner, nodeColor=<rgba 1 0.6328 0 1>}) {constB=0.0855}) {constMax=0.4, constMin=1}
  n111 = Distance(A=AppendVector(A=Add(A=@n107) {constB=0.31}, B=@n108), B=Constant2Vector() {g=0.5, r=0.5})
  n112 = SmoothStep(Value=Divide(A=NamedRerouteDeclaration(None=@n111) {name=Radial UV_TrailOuter, nodeColor=<rgba 1 0.6328 0 1>}) {constB=0.0855}) {constMax=0.4, constMin=1}
  n113 = NamedRerouteUsage()
  n114 = Step(Y=NamedRerouteUsage(), X=Divide(B=@n113) {constA=0.045}) {constX=0.06}
  n115 = Add(
   A=NamedRerouteDeclaration(None=Add(A=@n110, B=@n112)) {name=TrailInnerCurves, nodeColor=<rgba 0 0.9766 1 1>}
   B=Multiply(A=NamedRerouteUsage(), B=@n114)
  )
  n116 = Step(Y=NamedRerouteUsage(), X=Divide(B=@n113) {constA=0.085}) {constX=0.092}
  n117 = SmoothStep(Value=Divide(A=NamedRerouteUsage()) {constB=0.1}) {constMax=0.2, constMin=1}
  n118 = SmoothStep(Value=Divide(A=NamedRerouteUsage()) {constB=0.1}) {constMax=0.2, constMin=1}
  n119 = Multiply(A=Subtract(A=Add(A=@n117, B=@n118), B=NamedRerouteUsage()), B=@n106)
  n120 = Add(A=Multiply(A=Subtract(A=@n116, B=@n114), B=NamedRerouteUsage()), B=Saturate(None=Multiply(A=@n119) {constB=10}))
  NamedRerouteDeclaration(None=Saturate(None=Add(A=@n115, B=@n120))) {name=TrailEndOpacity, nodeColor=<rgba 0.3828 0 1 1>}
  n121 = NamedRerouteUsage()
  n122 = Saturate(None=@n115)
  NamedRerouteDeclaration(None=Add(A=Multiply(A=@n121, B=@n122), B=@n120)) {name=TrailEndColor, nodeColor=<rgba 0 1 0.0625 1>}
  n123 = NamedRerouteUsage()
  NamedRerouteDeclaration(None=Distance(A=AppendVector(A=@n123, B=@n2), B=Constant2Vector() {g=0.5, r=0.5})) {name=Radial UV, nodeColor=<rgba 1 0.6328 0 1>}
  n124 = NamedRerouteUsage()
  n125 = NamedRerouteUsage()
  n126 = Add(A=Step(Y=NamedRerouteUsage()) {constX=0.05}, B=OneMinus(None=Step(Y=@n125) {constX=0.95}))
  n127 = Multiply(A=Multiply(A=NamedRerouteUsage(), B=OneMinus(None=@n126))) {constB=0.025}
  n128 = OneMinus(None=MaterialFunctionCall!(Width=Constant() {r=0.5}, Sharpness=Constant(), Offset, Direction Switch=StaticBool() {value=true}, Compare, Input Coordinates) {materialFunction=/Engine/Functions/Engine_MaterialFunctions03/Procedurals/GeneratedBand}.Result)
  n129 = Add(A=@n126, B=Multiply(A=Power(Base=@n128)) {constB=0.5})
  n130 = Add(A=@n127, B=LinearInterpolate(Alpha=Saturate(None=@n129)) {constA=0.025, constB=0.5})
  n131 = OneMinus(None=Multiply(A=@n103, B=Step(X=@n102) {constX=0.1, constY=0.11}))
  n132 = Multiply(A=Power(Base=OneMinus(None=@n102)) {constExponent=1}) {constB=0.15}
  n133 = LinearInterpolate(Alpha=Saturate(None=Add(A=@n131, B=@n132))) {constA=0.1, constB=0.5}
  n134 = Multiply(A=@n104, B=LinearInterpolate(A=Add(A=@n133, B=@n127), B=@n130, Alpha=@n99))
  NamedRerouteDeclaration(None=LinearInterpolate(A=Multiply(A=@n124, B=@n130), B=Multiply(A=@n134, B=@n124), Alpha=NamedRerouteUsage())) {name=RollsAndSwellsColor, nodeColor=<rgba 0 1 0.1094 1>}
  NamedRerouteDeclaration(None=@n91) {name=Top UV Mask, nodeColor=<rgba 0.3125 0 1 1>}
  n135 = ScreenPosition()
  NamedRerouteDeclaration(None=SmoothStep(
    Min=ScalarParameter() {defaultValue=0.6}
    Max=ScalarParameter() {defaultValue=0.9, parameterName=Param_1}
    Value=ComponentMask(None=@n135.ViewportUV) {g=true}
   )) {name=Screen Space Mask, nodeColor=<rgba 0 0 0 1>}
  n136 = TextureCoordinate()
  n137 = Subtract(A=@n136) {constB=0.5}
  n138 = OneMinus(None=Multiply(A=Abs(None=@n137)) {constB=2})
  n139 = Multiply(
   A=SmoothStep(Value=ComponentMask(None=@n138) {r=true}) {constMax=0.03, constMin=0.01}
   B=SmoothStep(Value=ComponentMask(None=@n138) {g=true}) {constMax=0.03, constMin=0.01}
  )
  n140 = Constant() {r=-0.125}
  n141 = SmoothStep(Value=ComponentMask(None=MaterialFunctionCall(UVs=@n136, "Rotation Angle (0-1)"=@n140) {materialFunction=&CustomRotator}.Rotated Values) {r=true}) {constMax=0.51, constMin=0.49}
  n142 = Multiply(A=@n141, B=OneMinus(None=@n141))
  NamedRerouteDeclaration(None=Add(A=OneMinus(None=@n139), B=Multiply(A=@n142) {constB=5})) {name=DEBUG Wireframe, nodeColor=<rgba 0.1016 0 1 1>}
  n143 = NamedRerouteUsage()
  n144 = NamedRerouteUsage()
  n145 = Add(A=Multiply(A=OneMinus(None=@n143), B=Saturate(None=@n144)), B=NamedRerouteUsage())
  NamedRerouteDeclaration(None=LinearInterpolate(A=Saturate(None=@n145), B=NamedRerouteUsage(), Alpha=NamedRerouteUsage())) {name=Rounded End, nodeColor=<rgba 0.0547 0 1 1>}
  NamedRerouteUsage()
  ObjectBounds()
  Power(
   Base=MaterialFunctionCall(Width=Constant() {r=0.5}, Sharpness=Constant(), Direction Switch=StaticBool() {value=true}) {materialFunction=&GeneratedBand}.Result
   Exp=Constant() {r=1.5}
  )
  Step(Y=@n92) {constX=0.87}
  TextureCoordinate()
 ]
}
